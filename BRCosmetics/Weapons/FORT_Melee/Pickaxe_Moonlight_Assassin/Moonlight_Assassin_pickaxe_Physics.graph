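# Physics asset for the Moonlight Assassin pickaxe, a Fortnite melee cosmetic with dynamic moon pieces.
# A large capsule covers the two-handed handle, and small spheres cover the crescent, five moon rocks and the small-rock cluster.
# Every piece is constrained directly to the handle in a star layout, with no chaining.
# The crescent and small-rock cluster have locked swing and twist, while the five rocks carry small angular limit values.
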PhysicsAsset /BRCosmetics/Weapons/FORT_Melee/Pickaxe_Moonlight_Assassin/Moonlight_Assassin_pickaxe_Physics
bodies {
 Dyn_crescent [{shapeType=Sphere, radius=9.183, center=<0 -0.3516 -0.4745>}]
 Dyn_moon_rock1 [{shapeType=Sphere, radius=3.0399, center=<0.1486 0.0989 0.2965>}]
 Dyn_moon_rock2 [{shapeType=Sphere, radius=3.3895, center=<0.029 0.0243 0.7028>}]
 Dyn_moon_rock3 [{shapeType=Sphere, radius=5.4443, center=<-0.4468 -0.4782 -0.4825>}]
 Dyn_moon_rock4 [{shapeType=Sphere, radius=2.7902, center=<0.3469 -0.0408 -0.8062>}]
 Dyn_moon_rock5 [{shapeType=Sphere, radius=2.9772, center=<0.2722 -0.0465 -0.3229>}]
 Dyn_moon_smallrocks [{shapeType=Sphere, radius=4.1498, center=<-0.0324 0.3322 -0.8935>}]
 Melee_TwoHanded_Handle [{shapeType=Capsule, length=126.9698, radius=40.3613, center=<-0.7477 -1.707 36.7397>}]
}
constraints [
 {
  bone1Name=Dyn_moon_smallrocks
  bone2Name=Melee_TwoHanded_Handle
  swing1LimitDegrees=45
  swing1Motion=Locked
  swing2LimitDegrees=45
  swing2Motion=Locked
  twistLimitDegrees=45
  twistMotion=Locked
 }
 {bone1Name=Dyn_moon_rock1, bone2Name=Melee_TwoHanded_Handle, swing1LimitDegrees=5, swing2LimitDegrees=5, twistLimitDegrees=5}
 {bone1Name=Dyn_moon_rock2, bone2Name=Melee_TwoHanded_Handle, swing1LimitDegrees=5, swing2LimitDegrees=5, twistLimitDegrees=5}
 {bone1Name=Dyn_moon_rock3, bone2Name=Melee_TwoHanded_Handle, swing1LimitDegrees=5, swing2LimitDegrees=5, twistLimitDegrees=5}
 {bone1Name=Dyn_moon_rock4, bone2Name=Melee_TwoHanded_Handle, swing1LimitDegrees=5, swing2LimitDegrees=5, twistLimitDegrees=5}
 {bone1Name=Dyn_moon_rock5, bone2Name=Melee_TwoHanded_Handle, swing1LimitDegrees=5, swing2LimitDegrees=5, twistLimitDegrees=5}
 {
  bone1Name=Dyn_crescent
  bone2Name=Melee_TwoHanded_Handle
  swing1LimitDegrees=45
  swing1Motion=Locked
  swing2LimitDegrees=45
  swing2Motion=Locked
  twistLimitDegrees=45
  twistMotion=Locked
 }
]
