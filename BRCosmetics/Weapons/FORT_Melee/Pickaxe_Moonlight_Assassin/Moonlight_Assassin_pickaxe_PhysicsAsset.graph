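# Physics asset for the Moonlight Assassin two-handed pickaxe, giving collision shapes and joints to its handle and dangling dynamic bones.
# A large capsule covers the handle, a smaller one covers the spiral, and boxes cover the ribbon and two string segments.
# The spiral's swing and twist are locked to the handle.
# The ribbon joins the handle directly, while the strings form a two-link chain from it.
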PhysicsAsset /BRCosmetics/Weapons/FORT_Melee/Pickaxe_Moonlight_Assassin/Moonlight_Assassin_pickaxe_PhysicsAsset
bodies {
 Dyn_ribbon [{shapeType=Box, extentX=5.837, extentY=2.7052, extentZ=4.6441, center=<3.7144 0.0183 -0.1678>, rotation=<rot 0 -178.9303 -82.4152>}]
 Dyn_spiral [{shapeType=Capsule, length=14.8275, radius=14.1785, center=<0.376 0.2255 2.572>, rotation=<rot 0 0 -179.9999>}]
 Dyn_string01 [{shapeType=Box, extentX=0.2051, extentY=0.85, extentZ=7.1654, center=<3.3651 -0.0524 -0.143>, rotation=<rot 0 -90.7456 -84.291>}]
 Dyn_string02 [{shapeType=Box, extentX=0.2319, extentY=1.0233, extentZ=10.6237, center=<4.8872 -0.1807 -0.542>, rotation=<rot 0 -92.4115 -84.0875>}]
 Melee_TwoHanded_Handle [{shapeType=Capsule, length=117, radius=42.5, center=<0.8426 -1.1396 39.8525>}]
}
constraints [
 {
  bone1Name=Dyn_spiral
  bone2Name=Melee_TwoHanded_Handle
  swing1LimitDegrees=45
  swing1Motion=Locked
  swing2LimitDegrees=45
  swing2Motion=Locked
  twistLimitDegrees=45
  twistMotion=Locked
 }
 {bone1Name=Dyn_string01, bone2Name=Melee_TwoHanded_Handle, swing1LimitDegrees=45, swing2LimitDegrees=45, twistLimitDegrees=15}
 {bone1Name=Dyn_string02, bone2Name=Dyn_string01, swing1LimitDegrees=45, swing2LimitDegrees=45, twistLimitDegrees=15}
 {bone1Name=Dyn_ribbon, bone2Name=Melee_TwoHanded_Handle, swing1LimitDegrees=45, swing2LimitDegrees=45, twistLimitDegrees=45}
]
