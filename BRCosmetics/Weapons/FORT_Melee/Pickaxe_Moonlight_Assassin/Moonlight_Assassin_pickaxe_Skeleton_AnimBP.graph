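# Moonlight Assassin pickaxe AnimBlueprint that adds secondary motion to its ribbon, string, spiral, moon rock and crescent bones.
# Update builds sine/cosine bob vectors and derives RibbonAlpha from player character 0's forward speed, floored at 0.37.
# RigidBody plus additive bone offsets forms a cached pose, blended by bool against plain RigidBody.
# Per-bone layers add further RigidBody sims, ribbon weighted by RibbonAlpha; AnimDynamics and PoseDriver nodes sit unconnected.
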AnimBlueprint /BRCosmetics/Weapons/FORT_Melee/Pickaxe_Moonlight_Assassin/Moonlight_Assassin_pickaxe_Skeleton_AnimBP
target_skeleton=&/Moonlight_Assassin_pickaxe_Skeleton
graphs [
 AnimGraph {nodes [
   n1 = Anim Dynamics - Bone: Dyn_moon_smallrocks(Alpha=1, ExternalForce="0,0,0") {
    bGravityOverrideInSimSpace=true
    alphaBoolBlend {blendOption=Linear}
    boundBone {boneName=Dyn_moon_smallrocks}
    physicsBodyDefinitions [{
      sphereCollisionRadius=0
      boundBone {boneName=Dyn_moon_smallrocks}
      boxExtents=<10 5.4482 12.0492>
      constraintSetup {angularTargetAxis=AxisZ, twistAxis=AxisZ, angularLimitsMax=<0 0 0>, angularLimitsMin=<0 0 0>, angularTarget=<0 0 0>}
      localJointOffset=<-10 -1.814 -4.5614>
     }]
   }
   Anim Dynamics - Bone: Dyn_moon_rock1(ComponentPose=@n1, Alpha=1, ExternalForce="0,0,0") {
    bGravityOverrideInSimSpace=true
    alphaBoolBlend {blendOption=Linear}
    boundBone {boneName=Dyn_moon_rock1}
    physicsBodyDefinitions [{
      sphereCollisionRadius=0
      boundBone {boneName=Dyn_moon_rock1}
      boxExtents=<50 50 50>
      constraintSetup {coneAngle=10, twistAxis=AxisY, angularLimitsMax=<50 50 50>, angularLimitsMin=<0 0 0>, angularTarget=<10 10 10>}
      localJointOffset=<-10 -1 -1>
     }]
   }
   Pose Driver - Source: b_AR_Root!(SourcePose:Pose Link Structure) -> (Pose:Pose Link Structure) {
    is=title AnimGraphNode_PoseDriver
    evalSpaceBone {boneName=Dyn_string01}
    poseTargets [{
      bApplyCustomCurve=false
      bIsHidden=false
      distanceMethod=Quaternion
      drivenName=None
      functionType=Linear
      targetScale=1
      boneTransforms [{targetRotation=<rot 0 0 0>, targetTranslation=<0 0 0>}]
      customCurve {defaultValue=3.4028234663852886e+38, postInfinityExtrap=RCCE_Constant, preInfinityExtrap=RCCE_Constant, keys []}
      targetRotation=<rot 0 0 0>
     }]
    sourceBones [{boneName=b_AR_Root}]
   }
   n2 = RigidBody(ComponentPose=Local To Component(LocalPose=Use cached pose 'GraphPose'()), ExternalForce="0,0,1000") {
    bOverrideWorldGravity=true
    overridePhysicsAsset=&/Moonlight_Assassin_pickaxe_Physics
    alphaBoolBlend {blendOption=Linear}
    componentLinearAccScale=<0.25 0.25 0.25>
    componentLinearVelScale=<0.25 0.25 0.25>
    overrideWorldGravity {z=-386}
   }
   n3 = RigidBody(ComponentPose=Local To Component(LocalPose=Use cached pose 'GraphPose'()), ExternalForce="15,-15,10") {
    bOverrideWorldGravity=true
    overridePhysicsAsset=&/Moonlight_Assassin_pickaxe_PhysicsAsset
    alphaBoolBlend {blendOption=Linear}
    componentLinearAccScale=<1 1 1>
    componentLinearVelScale=<1 1 1>
    overrideWorldGravity {z=-980}
   }
   n4 = RigidBody(ComponentPose=Local To Component(LocalPose=Use cached pose 'GraphPose'())) {
    bOverrideWorldGravity=true
    overridePhysicsAsset=&Moonlight_Assassin_pickaxe_PhysicsAsset
    alphaBoolBlend {blendOption=Linear}
    componentLinearAccScale=<1 1 1>
    componentLinearVelScale=<1 1 1>
    overrideWorldGravity {z=-980}
   }
   n5 = Layered blend per bone!(
    BasePose=Blend Poses by bool(BlendPose_0=Component To Local(ComponentPose=@n2), BlendPose_1=Use cached pose 'Transforms'(), BlendTime_0=0.25, BlendTime_1=0.25) {blendType=Linear}
    BlendPoses_0=Component To Local(ComponentPose=@n4)
    BlendPoses_1:Pose Link Structure=Component To Local(ComponentPose=@n3)
    BlendWeights_0=Get RibbonAlpha() -> (RibbonAlpha:double)
    BlendWeights_1:float=1
   ) -> (Pose) {
    blendPoses [{linkId=-1, sourceLinkId=-1}, {linkId=-1, sourceLinkId=-1}]
    blendWeights [1,1]
    layerSetup [
     {branchFilters [{blendDepth=1, boneName=Dyn_ribbon}, {blendDepth=1, boneName=Dyn_Spiral}, {blendDepth=1, boneName=Dyn_string010}]}
     {branchFilters [{blendDepth=1, boneName=Dyn_string01}]}
    ]
   }
   Output Pose(Result=@n5) {layerGroup=None}
   AnimGraphNode_SubInput_0 = Input Pose()
   Save cached pose 'GraphPose'(Pose=@AnimGraphNode_SubInput_0) {cacheName=GraphPose}
   n6 = RigidBody(ComponentPose=Local To Component(LocalPose=Use cached pose 'GraphPose'())) {
    bForceDisableCollisionBetweenConstraintBodies=true
    bOverrideWorldGravity=true
    overridePhysicsAsset=&Moonlight_Assassin_pickaxe_Physics
    alphaBoolBlend {blendOption=Linear}
    componentLinearAccScale=<0.5 0.5 0.5>
    componentLinearVelScale=<0.5 0.5 0.5>
    overrideWorldGravity {z=-980}
   }
   n7 = "Transform (Modify) Bone - Bone: Dyn_spiral"(ComponentPose=@n6, Alpha=1, Translation=Get ShardFloatA() -> (ShardFloatA:Vector), Rotation="0,0,0", Scale="1,1,1") {translationMode=BMM_Additive, translationSpace=BCS_BoneSpace, alphaBoolBlend {blendOption=Linear}, boneToModify {boneName=Dyn_spiral}}
   n8 = "Transform (Modify) Bone - Bone: Dyn_moon_smallrocks"(ComponentPose=@n7, Alpha=1, Translation=Get ShardFloatB() -> (ShardFloatB:Vector), Rotation="0,0,0", Scale="1,1,1") {translationMode=BMM_Additive, alphaBoolBlend {blendOption=Linear}, boneToModify {boneName=Dyn_moon_smallrocks}}
   n9 = Get ShardFloatC() -> (ShardFloatC:Vector)
   n10 = "Transform (Modify) Bone - Bone: Dyn_moon_rock3"(
    ComponentPose="Transform (Modify) Bone - Bone: Dyn_moon_rock2"(
     ComponentPose="Transform (Modify) Bone - Bone: Dyn_moon_rock1"(ComponentPose=@n8, Alpha=1, Translation=@n9, Rotation="0,0,0", Scale="1,1,1") {translationMode=BMM_Additive, alphaBoolBlend {blendOption=Linear}, boneToModify {boneName=Dyn_moon_rock1}}
     Alpha=1
     Translation=Get ShardFloatA()
     Rotation="0,0,0"
     Scale="1,1,1"
    ) {translationMode=BMM_Additive, alphaBoolBlend {blendOption=Linear}, boneToModify {boneName=Dyn_moon_rock2}}
    Alpha=1
    Translation=Get ShardFloatB()
    Rotation="0,0,0"
    Scale="1,1,1"
   ) {translationMode=BMM_Additive, alphaBoolBlend {blendOption=Linear}, boneToModify {boneName=Dyn_moon_rock3}}
   n11 = Get ShardFloatC()
   n12 = "Transform (Modify) Bone - Bone: Dyn_crescent"(
    ComponentPose="Transform (Modify) Bone - Bone: Dyn_moon_rock5"(
     ComponentPose="Transform (Modify) Bone - Bone: Dyn_moon_rock4"(ComponentPose=@n10, Alpha=1, Translation=@n11, Rotation="0,0,0", Scale="1,1,1") {translationMode=BMM_Additive, alphaBoolBlend {blendOption=Linear}, boneToModify {boneName=Dyn_moon_rock4}}
     Alpha=1
     Translation=Get ShardFloatA()
     Rotation="0,0,0"
     Scale="1,1,1"
    ) {translationMode=BMM_Additive, alphaBoolBlend {blendOption=Linear}, boneToModify {boneName=Dyn_moon_rock5}}
    Alpha=1
    Translation=Get ShardFloatB()
    Rotation="0,0,0"
    Scale="1,1,1"
   ) {translationMode=BMM_Additive, alphaBoolBlend {blendOption=Linear}, boneToModify {boneName=Dyn_crescent}}
   n13 = Get ShardFloatC()
   Save cached pose 'Transforms'(Pose=Component To Local(ComponentPose="Transform (Modify) Bone - Bone: Dyn_crescent"(ComponentPose=@n12, Alpha=1, Translation=@n13, Rotation="0,0,0", Scale="1,1,1") {translationMode=BMM_Additive, alphaBoolBlend {blendOption=Linear}, boneToModify {boneName=Dyn_crescent}})) {cacheName=Transforms}
   OR Boolean(A=Get IsRunning() -> (IsRunning:bool), B=Get SwingBlend() -> (SwingBlend:bool))
   Get!() -> (TargetLoc_X:double, TargetLoc_Y:double, TargetLoc_Z:double)
   Get NewVar_1() -> (NewVar_1:bool)
   Get bIsFalling()
   Get!() -> (Push Side to Side_X:double, Push Side to Side_Y:double, Push Side to Side_Z:double)
   Get bIsJumping()
   Get pawn_speed_alpha() -> (pawn_speed_alpha:double)
  ]}
 EventGraph {nodes [
   Event BlueprintUpdateAnimation()
   n1 = GetGameTimeInSeconds()
   n2 = "Sin (Radians)"(A=float * float(A=float * float(A=1, B=1), B=@n1))
   n3 = "Sin (Radians)"(A=float * float(A=float * float(A=1, B=2), B=@n1))
   n4 = "Sin (Radians)"(A=float * float(A=float * float(A=1, B=1.5), B=@n1))
   > Set ShardFloatA(ShardFloatA:Vector=MakeVector(X=float * float(A=@n4, B=0.3), Y=float * float(A=@n2, B=0.56), Z=float * float(A=@n3, B=0.22))) -> (Output_Get:Vector)
   n5 = "Cos (Radians)"!(A:double=float * float(A=float * float(A=1, B=1.5), B=@n1)) -> (ReturnValue:double)
   n6 = "Cos (Radians)"(A=float * float(A=float * float(A=1, B=1.78), B=@n1))
   n7 = "Cos (Radians)"(A=float * float(A=float * float(A=1, B=1.15), B=@n1))
   > Set ShardFloatB(ShardFloatB:Vector=MakeVector(X=float * float(A=@n5, B=0.6), Y=float * float(A=@n6, B=0.725), Z=float * float(A=@n7, B=0.3))) -> (Output_Get:Vector)
   n8 = "Sin (Radians)"(A=float * float(A=float * float(A=1, B=1.5), B=@n1))
   n9 = "Sin (Radians)"(A=float * float(A=float * float(A=1, B=2), B=@n1))
   n10 = "Sin (Radians)"(A=float * float(A=float * float(A=1, B=1), B=@n1))
   > n11 = Set ShardFloatC(ShardFloatC:Vector=MakeVector(X=float * float(A=@n8, B=0.3), Y=float * float(A=@n10, B=0.6), Z=float * float(A=@n9, B=0.25))) -> (Output_Get:Vector)
   Set Motion_FromZ(execute=[@n30.then, @n11.then], Motion_FromZ:double=Get!() -> (VectorZ_X:double, VectorZ_Y:double, VectorZ_Z:double).VectorZ_Z) -> (Output_Get:double)
   n12 = "To Vector (Float)"!(InDouble:double=float * float(A=Get Motion_FromZ() -> (Motion_FromZ:double), B=200)) -> (ReturnValue:Vector)
   > Set Push Side to Side(Push Side to Side:Vector=@n12) -> (Output_Get:Vector)
   > Set bUpdateOwnerAnimInputProperty(bUpdateOwnerAnimInputProperty:bool=false) -> (Output_Get:bool)
   n13 = GetPlayerCharacter!(PlayerIndex:int=0) -> (ReturnValue:Character Object Reference)
   > Is Valid(InputObject=@n13)
   n14 = GetVelocity!(self:Actor Object Reference=@n13) -> (ReturnValue:Vector)
   n15 = Get Actor Rotation!(self:Actor Object Reference=@n13) -> (ReturnValue:Rotator) {is=title K2Node_CallFunction}
   n16 = BreakVector(InVec=vector * vector(A:Vector=Get Rotation X Vector!(InRot:Rotator=@n15) -> (ReturnValue:Vector) {is=title K2Node_CallFunction}, B:Vector=@n14) -> (ReturnValue:Vector) {is=title K2Node_CommutativeAssociativeBinaryOperator})
   n17 = float + float(A=@n16, B=@n16.Y) {is=title K2Node_CommutativeAssociativeBinaryOperator}
   n18 = "Not Equal (Float)"!(A:double=BreakVector(InVec=@n14).Z, B:double=0) -> (ReturnValue:bool) {is=title K2Node_CallFunction}
   > n19 = Set pawn_speed_alpha(pawn_speed_alpha:double=float * float(A=MapRangeClamped(Value=@n17, InRangeA=0, InRangeB=500), B=SelectFloat(A=0, B=1, bPickA=@n18))) -> (Output_Get:double)
   > n20 = Sequence()
   n21 = GetPartSkeletalMeshComponent()
   > Set SwingBlend(SwingBlend:bool="float >= float"(
     A=Get Distance Between Two Sockets!(
      Component:Skeletal Mesh Component Object Reference=@n21
      SocketOrBoneNameA:Name=axeTop
      SocketSpaceA:ERelativeTransformSpace Enum=RTS_Actor
      SocketOrBoneNameB:Name=axeSwing
      SocketSpaceB:ERelativeTransformSpace Enum=RTS_Actor
      bRemapRange:bool=false
      InRangeMin:float=4
      InRangeMax:float=45
      OutRangeMin:float=0
      OutRangeMax:float=1
     ) -> (ReturnValue:float) {is=title K2Node_CallFunction}
     B=0.5
    ) {is=title K2Node_CallFunction}) -> (Output_Get:bool)
   n22 = Branch(
    execute=@n20.then_1
    Condition=OR Boolean!(
     A=Get bIsJumping()
     B="float <= float"!(A:double=@n19, B:double=0.37) -> (ReturnValue:bool) {is=title K2Node_CallFunction}
     C:bool=Get bIsFalling()
    ) -> (ReturnValue)
   )
   > Set RibbonAlpha(RibbonAlpha:double=0.37) -> (Output_Get:double)
   > Set IsRunning(IsRunning:bool=false) -> (Output_Get:bool)
   Set RibbonAlpha(execute=@n22.else, RibbonAlpha=@n19)
   > Set IsRunning(IsRunning=true)
   n23 = OR Boolean(
    A="float > float"(A=Get Xelocity() -> (Xelocity:double), B=650) {is=title K2Node_CallFunction}
    B="float > float"(A=Get Yelocity() -> (Yelocity:double), B=650)
    C="float > float"(A=Get Zelocity() -> (Zelocity:double), B=650)
   )
   n24 = Branch(Condition=@n23)
   > Set IsRunning(IsRunning=true)
   Set IsRunning(execute=@n24.else, IsRunning=false)
   n25 = GetComponentVelocity!(self:Scene Component Object Reference=GetPartSkeletalMeshComponent(self=Self-Reference())) -> (ReturnValue_X:double, ReturnValue_Y:double, ReturnValue_Z:double)
   Set Xelocity(Xelocity:double=@n25) -> (Output_Get:double)
   > Set Yelocity(Yelocity:double=@n25.ReturnValue_Y) -> (Output_Get:double)
   > Set Zelocity(Zelocity:double=@n25.ReturnValue_Z) -> (Output_Get:double)
   n26 = GetGameTimeInSeconds()
   n27 = "Sin (Degrees)"!(A:double=float * float(A=@n26, B=100)) -> (ReturnValue:double)
   Set VectorX(VectorX:Vector=MakeVector(X=float * float(A=@n27, B=0.01), Y=0)) -> (Output_Get:Vector)
   n28 = "Sin (Degrees)"(A=float * float(A=@n26, B=125))
   > Set VectorY(VectorY:Vector=MakeVector(Y=float * float(A=@n28, B=0.01))) -> (Output_Get:Vector)
   n29 = "Sin (Degrees)"(A=float * float(A=@n26, B=75))
   > n30 = Set VectorZ(VectorZ:Vector=MakeVector(Y=0, Z=float * float(A=@n29, B=0.333))) -> (Output_Get:Vector)
   PrintString!(execute, InString="To String (Float)"(InDouble=Get RibbonAlpha()), bPrintToScreen, bPrintToLog, TextColor, Duration=0.03, Key) -> (then)
   n31 = Get Mesh(self=@n13)
   GetSocketLocation!(self:Scene Component Object Reference=@n31, InSocketName:Name=None) -> (ReturnValue:Vector)
   GetCurrentJointAngles!(execute, self:Skeletal Mesh Component Object Reference=Get Mesh(self=@n13), InBoneName:Name=None) -> (then, Swing1Angle:float, TwistAngle:float, Swing2Angle:float)
   NOT Boolean(A=@n23)
   "float < float"!(A:double=@n19, B:double=0.37) -> (ReturnValue:bool) {is=title K2Node_CallFunction}
   PrintString(InString="To String (Vector)"!(InVec:Vector=@n12) -> (ReturnValue:String), Duration=0.03)
   n32 = TransformToBoneSpace!(
    self:Skinned Mesh Component Object Reference=@n31
    BoneName:Name=weapon_r
    InPosition:Vector=GetSocketLocation(self=@n31, InSocketName=weapon_r_socket)
    InRotation:Rotator="0, 0, 0"
   ) -> (OutPosition:Vector, OutRotation:Rotator)
   Set TargetLoc(TargetLoc:Vector=vector * vector(A=@n32, B=@n32)) -> (Output_Get:Vector)
  ]}
]
variables [
 IsRunning {type=boolean, default=true}
 Motion_FromZ {type=number}
 NewVar_1 {type=boolean}
 pawn_speed_alpha {type=number}
 Push Side to Side {type=Vector}
 RibbonAlpha {type=number}
 ShardFloatA {type=Vector}
 ShardFloatB {type=Vector}
 ShardFloatC {type=Vector}
 SwingBlend {type=boolean}
 TargetLoc {type=Vector}
 VectorX {type=Vector}
 VectorY {type=Vector}
 VectorZ {type=Vector}
 Xelocity {type=number}
 Yelocity {type=number}
 Zelocity {type=number}
]
